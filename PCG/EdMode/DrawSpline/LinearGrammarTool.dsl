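; PCG graph for the Draw Spline editor-mode tool that lays mesh modules along an input spline following a grammar string.
; Modules, optionally sized from mesh bounds, feed Subdivide Spline, whose points spawn as offset static meshes or display as debug boxes.
; Optionally, modules whose forward/up dot products with neighbours fall below a threshold become spline meshes.
; These use tangents resampled from a spline through averaged module joints, handling closed loops and optional roll.
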(graph LinearGrammarTool
  (params (Grammar :type String :default "A*") (Modules :type (Array (Struct /PCG/EdMode/Resources/PCGEdMode_MeshModuleInfo.PCGEdMode_MeshModuleInfo)) :default [(struct :StaticMesh /PCG/DebugObjects/PCG_AxisTripod.PCG_AxisTripod :Symbol A :Scalable true)]) (SpawnSplineMeshes :type Bool) (SplineMeshAngleThreshold :type Double :default 0.995) (SetSplineMeshRoll :type Bool) (SplineUpDir :type Vector :default (0.0 0.0 1.0)) (SplineRollDegrees :type Double) (DebugGrammar :type Bool))
  ;[0 0 0 0.5 32] Apply Grammar
  (bind AttributeMathsOp_32 (Attribute%20Maths%20Op :InputSource1 $Index :InputSource2 $Index :Operation Set :OutputTarget ModulesOrder :InA Modules :InB Modules))
  (bind AttributePartition_2 (Attribute%20Partition :PartitionAttributeSelectors [ForwardAxis]))
  (bind Loop_0 (Loop :Graph /PCG/EdMode/DrawSpline/Subgraphs/LinearGrammar_ForwardAxisLoop.LinearGrammar_ForwardAxisLoop))
  (bind MergeAttributes_1 (Merge%20Attributes))
  (bind AttributeSetToPoint_0 (Attribute%20Set%20To%20Point))
  (bind AttributeFilter_0 (Filter%20Attribute%20Elements :AttributeTypes (struct :Type Boolean) :bGenerateOutputDataEvenIfEmpty false :bUseConstantThreshold true :Operator Equal :TargetAttribute UseMeshBoundsAsModuleSize :ThresholdAttribute @Last))
  (bind BoundsFromMesh_1 (Bounds%20From%20Mesh :MeshAttribute StaticMesh))
  (bind AttributeMathsOp_18 (Attribute%20Maths%20Op :InputSource1 $LocalSize :InputSource2 ForwardAxisVector :Operation Multiply :OutputTarget MeshSize))
  (bind AttributeMathsOp_30 (Attribute%20Maths%20Op :InputSource1 Size :InputSource2 MeshSize.length :Operation Set))
  (bind PointToAttributeSet_0 (Point%20To%20Attribute%20Set))
  (bind MergeAttributes_0 (Merge%20Attributes))
  (bind SortAttributes_0 (Sort%20Attributes :bUseStableSort false :SortAttributes [(struct :InputSource ModulesOrder)]))
  (bind SubdivideSpline_0 (Subdivide%20Spline :bAcceptIncompleteSubdivision true :bForwardAttributesFromModulesInfo true :bModuleInfoAsInput true :bOutputScalableAttribute false :bOutputSizeAttribute false :GrammarSelection (struct :GrammarString "[C]{W,W1}*[C]") :ModuleHeight 0.0 :ModuleHeightAttribute ExtentsZ :ModulesInfo [(struct :Symbol Fence :bScalable true :DebugColor (1.0 0.0 0.0 1.0)) (struct :Symbol Door :Size 3e+02 :DebugColor (0.0 1.0 0.0 1.0)) (struct :Symbol C :Size 5e+01 :DebugColor (0.0 0.0 1.0 1.0))] :ModulesInfoAttributeNames (struct :bProvideScalable true) :ScalableAttributeName bScalable :GrammarString Grammar)) ; Module info is \nembedded in the node
  (bind Branch_0 (Branch :bOutputToB-param DebugGrammar))
  (bind ApplyScaleToBounds_2 (Apply%20Scale%20To%20Bounds))
  (bind AttributeMathsOp_7 (Attribute%20Maths%20Op :InputSource1 $Color :InputSource2 DebugColor :Operation Set))
  (bind AttributeMathsOp_8 (Attribute%20Maths%20Op :InputSource1 $Extents.yz :Operation Set))
  (bind Debug_1 (Debug))
  ;[0 0 0 0.5 32] Spline Mesh Spawning
  ;;[0 0 0 0.5] Prepare Spline Mesh Data: Create Points at the Start / End of each Module, on the spline
  (bind ResetPointCenter_0 (Reset%20Point%20Center :PointCenterLocation (0.0 0.5 0.5)))
  (bind AttributeFilter_6 (Filter%20Attribute%20Elements :AttributeTypes (struct :Type Integer32) :bUseConstantThreshold true :Operator Equal :TargetAttribute $Index :ThresholdAttribute @Last))
  (bind GetPointsCount_11 (Get%20Element%20Count))
  (bind AttributeMathsOp_12 (Attribute%20Maths%20Op :Operation Subtract))
  (bind ResetPointCenter_1 (Reset%20Point%20Center :PointCenterLocation (1.0 0.5 0.5)))
  (bind AttributeFilter_10 (Filter%20Attribute%20Elements :AttributeTypes (struct :Type Integer32) :Operator Lesser :TargetAttribute $Index :ThresholdAttribute @Last))
  (bind CreateAttribute_15 (Create%20Constant :AttributeTypes (struct :DoubleValue 0.5)))
  (bind AttributeTransformOp_14 (Attribute%20Transform%20Op :InputSource1 $Transform :InputSource2 $Transform :Operation Lerp))
  (bind Branch_1 (Branch))
  (bind MergePoints_17 (Merge%20Points))
  ;;;[0 0 0 0.5] Create Spline From Points and Sample it to get tangent attributes
  (bind CreateSpline_9 (Create%20Spline :ArriveTangentAttribute None :InterpTypeAttribute None :LeaveTangentAttribute None))
  (bind SplineSampler_10 (Spline%20Sampler :SamplerParams (struct :SubdivisionsPerSegment 0 :bComputeTangents true :bUnbounded true :PointSteepness 1.0)))
  ;;[0 0 0 0.5] Copy StaticMesh Attrib. from Segment to each Spline
  (bind CreateSpline_52 (Create%20Spline :bApplyCustomTangents true :InterpTypeAttribute None))
  (bind GetAttributeFromPointIndex_56 (Get%20Attribute%20From%20Point%20Index :bForceOutputAttributeToBeInElementsDomain true :InputSource StaticMesh))
  (bind CopyAttributes_58 (Copy%20Attributes :InputSource StaticMesh :OutputTarget @Data.StaticMesh))
  ;;[0 0 0 0.5] Create segments points pairs
  (bind AttributeFilter_5 (Filter%20Attribute%20Elements :AttributeTypes (struct :Type Integer32) :bUseConstantThreshold true :TargetAttribute $Index :ThresholdAttribute @Last))
  (bind MergePoints_7 (Merge%20Points))
  (bind GetPointsCount_5 (Get%20Element%20Count))
  (bind AttributeMathsOp_19 (Attribute%20Maths%20Op :Operation Subtract))
  (bind CopyAttributes_1 (Copy%20Attributes :InputSource $Index :OutputTarget SegmentID))
  (bind CopyAttributes_2 (Copy%20Attributes :InputSource $Index :OutputTarget SegmentID))
  (bind MergePoints_8 (Merge%20Points))
  (bind Branch_5 (Branch))
  (bind AttributeFilter_7 (Filter%20Attribute%20Elements :Operator Lesser :TargetAttribute SegmentID :ThresholdAttribute @Last))
  (bind AttributePartition_0 (Attribute%20Partition :PartitionAttributeSelectors [SegmentID]))
  ;;;[0 0 0 0.5] Keep only segments to be spawned as spline meshes
  (bind GetAttributeFromPointIndex_0 (Get%20Attribute%20From%20Point%20Index :bForceOutputAttributeToBeInElementsDomain true :InputSource SpawnAsSplineMesh))
  (bind MergeAttributes_2 (Merge%20Attributes))
  (bind CopyAttributes_6 (Copy%20Attributes :InputSource $Index :OutputTarget IndicesToKeep))
  (bind AttributeFilter_9 (Filter%20Attribute%20Elements :AttributeTypes (struct :Type Boolean :BoolValue true) :bGenerateOutputDataEvenIfEmpty false :bUseConstantThreshold true :Operator Equal :TargetAttribute SpawnAsSplineMesh :ThresholdAttribute @Last))
  (bind AttributeReduce_0 (Attribute%20Reduce :InputSource IndicesToKeep :Operation Join :OutputAttributeName IndicesToKeep))
  (bind FilterDataByIndex_0 (Filter%20Data%20By%20Index))
  ;;[0 0 0 0.5] SplineMesh Roll Fixes
  (bind Branch_14 (Branch :bOutputToB-param SetSplineMeshRoll))
  (bind CopyAttributes_5 (Copy%20Attributes :InputSource @Last :OutputTarget @Data.SplineUpDir :Source SplineUpDir))
  (bind CopyAttributes_4 (Copy%20Attributes :InputSource @Last :OutputTarget @Data.RollDegrees :Source SplineRollDegrees))
  (bind CopyAttributes_7 (Copy%20Attributes :InputSource @Last :OutputTarget @Data.RollDegreesEnd :Source SplineRollDegrees))
  ;;[0 0 0 0.5] Spawn Spline Meshes
  (bind Loop_2 (Loop :Graph /PCG/EdMode/DrawSpline/Subgraphs/LinearGrammar_SpawnSplineMeshLoop.LinearGrammar_SpawnSplineMeshLoop :SetSplineMeshRoll SetSplineMeshRoll))
  ;[0 0 0 0.5 32] Static Mesh Spawning
  ;;[0 0 0 0.5] Filter points for Spline Mesh Spawning based on a defined angle tolerance threshold.
  (bind Branch_12 (Branch :bOutputToB-param SpawnSplineMeshes))
  (bind GetPointsCount_2 (Get%20Element%20Count))
  (bind AttributeMathsOp_5 (Attribute%20Maths%20Op :Operation Subtract))
  (bind AttributeFilter_4 (Filter%20Attribute%20Elements :AttributeTypes (struct :Type Integer32) :Operator Equal :TargetAttribute $Index :ThresholdAttribute @Last))
  (bind MergePoints_1 (Merge%20Points))
  (bind AttributeVectorOp_1 (Attribute%20Vector%20Op :InputSource1 $Rotation.Forward :InputSource2 $Rotation.Forward :Operation Dot :OutputTarget dot))
  (bind AttributeFilter_102 (Filter%20Attribute%20Elements :AttributeTypes (struct :Type Integer32) :bUseConstantThreshold true :TargetAttribute $Index :ThresholdAttribute @Last))
  (bind MergePoints_104 (Merge%20Points))
  (bind AttributeVectorOp_107 (Attribute%20Vector%20Op :InputSource1 $Rotation.Forward :InputSource2 $Rotation.Forward :Operation Dot :OutputTarget dot))
  (bind AttributeMathsOp_6 (Attribute%20Maths%20Op :Operation Min))
  (bind AttributeVectorOp_2 (Attribute%20Vector%20Op :InputSource1 $Rotation.Up :InputSource2 $Rotation.Up :Operation Dot :OutputTarget dot))
  (bind AttributeVectorOp_3 (Attribute%20Vector%20Op :InputSource1 $Rotation.Up :InputSource2 $Rotation.Up :Operation Dot :OutputTarget dot))
  (bind AttributeMathsOp_20 (Attribute%20Maths%20Op :Operation Min))
  (bind AttributeMathsOp_21 (Attribute%20Maths%20Op :Operation Min))
  (bind AttributeCompareOp_3 (Attribute%20Compare%20Op :InputSource1 dot :Operation Less :OutputTarget SpawnAsSplineMesh :InB SplineMeshAngleThreshold))
  (bind AttributeFilter_8 (Filter%20Attribute%20Elements :AttributeTypes (struct :Type Boolean :DoubleValue 0.995) :bUseConstantThreshold true :Operator Equal :TargetAttribute SpawnAsSplineMesh :ThresholdAttribute @Last))
  ;;[0 0 0 0.5] Apply Mesh Pivot Offset
  (bind AttributePartition_3 (Attribute%20Partition :PartitionAttributeSelectors [ForwardAxis]))
  (bind Loop_1 (Loop :Graph /PCG/EdMode/DrawSpline/Subgraphs/LinearGrammar_ApplyOffsetsLoop.LinearGrammar_ApplyOffsetsLoop))
  ;;[0 0 0 0.5] Spawn Static Meshes
  (bind StaticMeshSpawner_41 (Static%20Mesh%20Spawner :MeshSelectorParameters (class:/Script/PCG.PCGMeshSelectorByAttribute :AttributeName StaticMesh :TemplateDescriptor (struct :OverrideMaterials [null] :BodyInstance (struct :CollisionEnabled NoCollision :SolverAsyncDeltaTime 0.016667 :CollisionProfileName NoCollision :CollisionResponses (struct :ResponseArray [(struct :Channel Visibility :Response ECR_Ignore) (struct :Channel Camera :Response ECR_Ignore)]) :bAutoWeld true))) :MeshSelectorType /Script/PCG.PCGMeshSelectorByAttribute :StaticMeshComponentPropertyOverrides [(struct :InputSource AllowCollision :PropertyTarget "bUseDefaultCollision")]))
  ;[0 0 0 0.5 32] Input
  (bind GetSplineData_1 (Get%20Spline%20Data))
  (bind RemoveEmptyData_0 (Remove%20Empty%20Data))
  (bind GetElementsCount_0 (Get%20Element%20Count))
  (bind Branch_7 (Branch))
  ;;[0 0 0 0.5] Extract Spline IsClosed
  (bind ExtractAttribute_0 (Extract%20Attribute%20at%20Index :bForceOutputAttributeToBeInElementsDomain true :InputSource @Data.$IsClosed))
  (reroute InputSpline Branch_7.Output%20B)
  (reroute IsClosedLoop ExtractAttribute_0)
  (reroute PointsForSplineMeshSpawning AttributeCompareOp_3)
  (reroute PointsFromGrammar Branch_0.Output%20A)
  (reroute PointsFromModulesStartEnds SplineSampler_10)
  (reroute SegmentsPointPairs FilterDataByIndex_0.InsideFilter)
  (connect Branch_0 Output%20B ApplyScaleToBounds_2 In)
  (connect AttributeMathsOp_21 Out AttributeCompareOp_3 InA)
  (connect AttributeSetToPoint_0 Out AttributeFilter_0 In)
  (connect AttributeMathsOp_5 Out AttributeFilter_4 Filter)
  (connect Branch_12 Output%20B AttributeFilter_4 In)
  (connect PointsFromModulesStartEnds Out AttributeFilter_5 In)
  (connect ResetPointCenter_0 Out AttributeFilter_6 In)
  (connect AttributeMathsOp_19 Out AttributeFilter_7 Filter)
  (connect Branch_5 Output%20A AttributeFilter_7 In)
  (connect AttributeCompareOp_3 Out AttributeFilter_8 In)
  (connect CopyAttributes_6 Out AttributeFilter_9 In)
  (connect AttributeMathsOp_12 Out AttributeFilter_10 Filter)
  (connect ResetPointCenter_1 Out AttributeFilter_10 In)
  (connect Branch_12 Output%20B AttributeFilter_102 In)
  (connect GetPointsCount_2 Out AttributeMathsOp_5 InA)
  (connect AttributeVectorOp_107 Out AttributeMathsOp_6 InA)
  (connect AttributeVectorOp_1 Out AttributeMathsOp_6 InB)
  (connect ApplyScaleToBounds_2 Out AttributeMathsOp_7 InA)
  (connect ApplyScaleToBounds_2 Out AttributeMathsOp_7 InB)
  (connect AttributeMathsOp_7 Out AttributeMathsOp_8 InA)
  (connect GetPointsCount_11 Out AttributeMathsOp_12 InA)
  (connect BoundsFromMesh_1 Out AttributeMathsOp_18 InA)
  (connect BoundsFromMesh_1 Out AttributeMathsOp_18 InB)
  (connect GetPointsCount_5 Out AttributeMathsOp_19 InA)
  (connect AttributeVectorOp_2 Out AttributeMathsOp_20 InA)
  (connect AttributeVectorOp_3 Out AttributeMathsOp_20 InB)
  (connect AttributeMathsOp_20 Out AttributeMathsOp_21 InA)
  (connect AttributeMathsOp_6 Out AttributeMathsOp_21 InB)
  (connect AttributeMathsOp_18 Out AttributeMathsOp_30 InA)
  (connect AttributeMathsOp_18 Out AttributeMathsOp_30 InB)
  (connect AttributeFilter_7 InsideFilter AttributePartition_0 In)
  (connect Branch_5 Output%20B AttributePartition_0 In)
  (connect AttributeMathsOp_32 Out AttributePartition_2 In)
  (connect AttributeFilter_8 InsideFilter AttributePartition_3 In)
  (connect Branch_12 Output%20A AttributePartition_3 In)
  (connect AttributeFilter_9 InsideFilter AttributeReduce_0 In)
  (connect MergeAttributes_1 Out AttributeSetToPoint_0 In)
  (connect AttributeFilter_6 OutsideFilter AttributeTransformOp_14 InA)
  (connect AttributeFilter_10 InsideFilter AttributeTransformOp_14 InB)
  (connect CreateAttribute_15 Out AttributeTransformOp_14 Ratio)
  (connect Branch_12 Output%20B AttributeVectorOp_1 InA)
  (connect MergePoints_1 Out AttributeVectorOp_1 InB)
  (connect Branch_12 Output%20B AttributeVectorOp_2 InA)
  (connect MergePoints_104 Out AttributeVectorOp_2 InB)
  (connect Branch_12 Output%20B AttributeVectorOp_3 InA)
  (connect MergePoints_1 Out AttributeVectorOp_3 InB)
  (connect Branch_12 Output%20B AttributeVectorOp_107 InA)
  (connect MergePoints_104 Out AttributeVectorOp_107 InB)
  (connect AttributeFilter_0 OutsideFilter BoundsFromMesh_1 In)
  (connect SubdivideSpline_0 Out Branch_0 In)
  (connect IsClosedLoop Out Branch_1 bOutputToB)
  (connect AttributeFilter_10 OutsideFilter Branch_1 In)
  (connect IsClosedLoop Out Branch_5 bOutputToB)
  (connect MergePoints_8 Out Branch_5 In)
  (connect GetElementsCount_0 Out Branch_7 bOutputToB)
  (connect RemoveEmptyData_0 Out Branch_7 In)
  (connect PointsFromGrammar Out Branch_12 In)
  (connect CopyAttributes_58 Out Branch_14 In)
  (connect PointsFromModulesStartEnds Out CopyAttributes_1 Source)
  (connect PointsFromModulesStartEnds Out CopyAttributes_1 Target)
  (connect MergePoints_7 Out CopyAttributes_2 Source)
  (connect MergePoints_7 Out CopyAttributes_2 Target)
  (connect CopyAttributes_5 Out CopyAttributes_4 Target)
  (connect Branch_14 Output%20B CopyAttributes_5 Target)
  (connect MergeAttributes_2 Out CopyAttributes_6 Source)
  (connect MergeAttributes_2 Out CopyAttributes_6 Target)
  (connect CopyAttributes_4 Out CopyAttributes_7 Target)
  (connect GetAttributeFromPointIndex_56 Attribute CopyAttributes_58 Source)
  (connect CreateSpline_52 Out CopyAttributes_58 Target)
  (connect IsClosedLoop Out CreateSpline_9 bClosedLoop)
  (connect MergePoints_17 Out CreateSpline_9 In)
  (connect SegmentsPointPairs Out CreateSpline_52 In)
  (connect AttributeMathsOp_8 Out Debug_1 In)
  (connect InputSpline Out ExtractAttribute_0 In)
  (connect AttributePartition_0 Out FilterDataByIndex_0 In)
  (connect AttributeReduce_0 Out FilterDataByIndex_0 SelectedIndices)
  (connect AttributePartition_0 Out GetAttributeFromPointIndex_0 In)
  (connect SegmentsPointPairs Out GetAttributeFromPointIndex_56 In)
  (connect RemoveEmptyData_0 Out GetElementsCount_0 In)
  (connect Branch_12 Output%20B GetPointsCount_2 In)
  (connect PointsFromModulesStartEnds Out GetPointsCount_5 In)
  (connect PointsForSplineMeshSpawning Out GetPointsCount_11 In)
  (connect AttributePartition_2 Out Loop_0 In)
  (connect AttributePartition_3 Out Loop_1 In)
  (connect Branch_14 Output%20A Loop_2 In)
  (connect CopyAttributes_7 Out Loop_2 In)
  (connect PointToAttributeSet_0 Out MergeAttributes_0 In)
  (connect Loop_0 Out MergeAttributes_1 In)
  (connect GetAttributeFromPointIndex_0 Attribute MergeAttributes_2 In)
  (connect AttributeFilter_4 InsideFilter MergePoints_1 In)
  (connect AttributeFilter_4 OutsideFilter MergePoints_1 In2)
  (connect AttributeFilter_5 InsideFilter MergePoints_7 In)
  (connect AttributeFilter_5 OutsideFilter MergePoints_7 In2)
  (connect CopyAttributes_1 Out MergePoints_8 In)
  (connect CopyAttributes_2 Out MergePoints_8 In2)
  (connect AttributeFilter_6 InsideFilter MergePoints_17 In)
  (connect AttributeTransformOp_14 Out MergePoints_17 In2)
  (connect Branch_1 Output%20A MergePoints_17 In3)
  (connect AttributeFilter_102 InsideFilter MergePoints_104 In)
  (connect AttributeFilter_102 OutsideFilter MergePoints_104 In2)
  (connect AttributeFilter_0 InsideFilter PointToAttributeSet_0 In)
  (connect AttributeMathsOp_30 Out PointToAttributeSet_0 In)
  (connect GetSplineData_1 Out RemoveEmptyData_0 In)
  (connect PointsForSplineMeshSpawning Out ResetPointCenter_0 In)
  (connect PointsForSplineMeshSpawning Out ResetPointCenter_1 In)
  (connect MergeAttributes_0 Out SortAttributes_0 In)
  (connect CreateSpline_9 Out SplineSampler_10 Spline)
  (connect Loop_1 Out StaticMeshSpawner_41 In)
  (connect InputSpline Out SubdivideSpline_0 In)
  (connect SortAttributes_0 Out SubdivideSpline_0 ModulesInfo))
