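; PCG graph for an editor draw-spline spacing tool that spawns weighted static meshes or actors at points spaced along a spline by count or distance.
; Sampled points receive weighted entries, then global and random transforms.
; A ray query from the top of the points' combined bounds feeds a projection for optional surface snapping and alignment.
; With an empty asset list, points are tagged and shown as debug previews instead of spawned.
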(graph SpacingTool
  (params (Count :type Int :default 10) (UseDistanceSamplingMode :type Bool) (Distance :type Double :default 1e+02) (FitToCurve :type Bool) (StartOffset :type Double) (EndOffset :type Double) (MaxRandomOffset :type Double) (SamplingSeed :type Int :default 2145617851) (StaticMeshes :type (Array (Struct /PCG/EdMode/Resources/PCGEdMode_WeightedStaticMeshReference.PCGEdMode_WeightedStaticMeshReference))) (SpawnActors :type Bool) (ActorClasses :type (Array (Struct /PCG/EdMode/Resources/PCGEdMode_WeightedClassReference.PCGEdMode_WeightedClassReference))) (GlobalPositionOffset :type Vector) (GlobalRotation :type Rotator) (AbsoluteGlobalRotation :type Bool) (GlobalScaleFactor :type Vector :default (1.0 1.0 1.0)) (RandomPositionOffsetMin :type Vector) (RandomPositionOffsetMax :type Vector) (RandomRotationMin :type Rotator) (RandomRotationMax :type Rotator) (RandomScaleFactorMin :type Vector :default (1.0 1.0 1.0)) (RandomScaleFactorMax :type Vector :default (1.0 1.0 1.0)) (UniformRandomScaleFactor :type Bool :default true) (RandomSeed :type Int :default 2118139241) (SnapToSurface :type Bool) (AlignToSurface :type Bool) (SnapToSurfaceMaxDistance :type Double :default 1e+03))
  (bind CreateConstant_0 (Create%20Constant :AttributeTypes (struct :Type Integer32 :Int32Value 2)))
  (bind AttributeMathsOp_0 (Attribute%20Maths%20Op :Operation Subtract :InB UseDistanceSamplingMode))
  (bind AttributeMathsOp_1 (Attribute%20Maths%20Op :Operation Divide :InA MaxRandomOffset))
  (bind GetSplineData_0 (Get%20Spline%20Data))
  (bind SplineSampler_0 (Spline%20Sampler :SamplerParams (struct :Mode NumberOfSamples :DistanceIncrement 5e+02 :bUnbounded true :PointSteepness 1.0) :bFitToCurve FitToCurve :DistanceIncrement Distance :EndOffset EndOffset :NumSamples Count :Seed-param SamplingSeed :StartOffset StartOffset))
  (bind Select_1 (Select :bUseInputB-param SpawnActors :Input%20A StaticMeshes :Input%20B ActorClasses))
  (bind RemoveEmptyData_1 (Remove%20Empty%20Data))
  (bind DataNum_1 (Data%20Count))
  (bind BoundsModifier_1 (Bounds%20Modifier :BoundsMax (5e+01 5e+01 5e+01) :BoundsMin (-5e+01 -5e+01 -5e+01) :Mode Set))
  (bind Branch_0 (Branch))
  (bind MatchAndSetAttributes_1 (Match%20And%20Set%20Attributes :bUseWeightAttribute true :WeightAttribute Weight))
  (bind AddTags_0 (Add%20Tags :TagsToAdd "ShowPreviewMesh"))
  (bind Branch_2 (Branch :bOutputToB-param SpawnActors))
  (bind BoundsFromMesh_1 (Bounds%20From%20Mesh :MeshAttribute StaticMesh))
  (bind TransformPoints_1 (Transform%20Points :bUniformScale false :bAbsoluteRotation-param AbsoluteGlobalRotation :OffsetMax-param GlobalPositionOffset :OffsetMin-param GlobalPositionOffset :RotationMax-param GlobalRotation :RotationMin-param GlobalRotation :ScaleMax-param GlobalScaleFactor :ScaleMin-param GlobalScaleFactor))
  (bind TransformPoints_2 (Transform%20Points :bUniformScale-param UniformRandomScaleFactor :OffsetMax-param RandomPositionOffsetMax :OffsetMin-param RandomPositionOffsetMin :RotationMax-param RandomRotationMax :RotationMin-param RandomRotationMin :ScaleMax-param RandomScaleFactorMax :ScaleMin-param RandomScaleFactorMin :Seed-param RandomSeed))
  (bind CombinePoints_0 (Combine%20Points :bUseFirstPointTransform false))
  (bind AttributeMathsOp_2 (Attribute%20Maths%20Op :InputSource1 $Position.z :InputSource2 $BoundsMax.Z))
  (bind ExtractAttribute_0 (Extract%20Attribute%20at%20Index :bForceOutputAttributeToBeInElementsDomain true :InputSource $Position))
  (bind WorldRayHitQuery_1 (World%20Ray%20Hit%20Query :QueryParams (struct :bOverrideDefaultParams true :RayLength 1e+03) :RayLength SnapToSurfaceMaxDistance))
  (bind Projection_1 (Projection :bKeepZeroDensityPoints true :ProjectionParams (struct :AttributeMode IncludeAttributes :AttributeMergeOperation SourceValue) :bProjectPositions SnapToSurface :bProjectRotations AlignToSurface))
  (bind FilterDataByTag_0 (Filter%20Data%20By%20Tag :SelectedTags "ShowPreviewMesh"))
  (bind Branch_1 (Branch :bOutputToB-param SpawnActors))
  (bind StaticMeshSpawner_1 (Static%20Mesh%20Spawner :MeshSelectorParameters (class:/Script/PCG.PCGMeshSelectorByAttribute :AttributeName StaticMesh :TemplateDescriptor (struct :bUseDefaultCollision true)) :MeshSelectorType /Script/PCG.PCGMeshSelectorByAttribute))
  (bind SpawnActor_1 (Spawn%20Actor :bSpawnByAttribute true :PostSpawnFunctionNames [None] :SpawnAttribute Class))
  (bind Debug_0 (Debug))
  (connect Branch_0 Output%20A AddTags_0 In)
  (connect CreateConstant_0 Out AttributeMathsOp_0 InA)
  (connect CombinePoints_0 Out AttributeMathsOp_2 InA)
  (connect CombinePoints_0 Out AttributeMathsOp_2 InB)
  (connect Branch_2 Output%20A BoundsFromMesh_1 In)
  (connect SplineSampler_0 Out BoundsModifier_1 In)
  (connect DataNum_1 Out Branch_0 bOutputToB)
  (connect BoundsModifier_1 Out Branch_0 In)
  (connect FilterDataByTag_0 OutsideFilter Branch_1 In)
  (connect MatchAndSetAttributes_1 Out Branch_2 In)
  (connect TransformPoints_2 Out CombinePoints_0 In)
  (connect RemoveEmptyData_1 Out DataNum_1 In)
  (connect FilterDataByTag_0 InsideFilter Debug_0 In)
  (connect AttributeMathsOp_2 Out ExtractAttribute_0 In)
  (connect Projection_1 Out FilterDataByTag_0 In)
  (connect Branch_0 Output%20B MatchAndSetAttributes_1 In)
  (connect Select_1 Out MatchAndSetAttributes_1 Match%20Data)
  (connect TransformPoints_2 Out Projection_1 In)
  (connect WorldRayHitQuery_1 Out Projection_1 Projection%20Target)
  (connect Select_1 Out RemoveEmptyData_1 In)
  (connect Branch_1 Output%20B SpawnActor_1 In)
  (connect AttributeMathsOp_1 Out SplineSampler_0 MaxRandomOffsetNormalized)
  (connect AttributeMathsOp_0 Out SplineSampler_0 Mode)
  (connect GetSplineData_0 Out SplineSampler_0 Spline)
  (connect Branch_1 Output%20A StaticMeshSpawner_1 In)
  (connect AddTags_0 Out TransformPoints_1 In)
  (connect BoundsFromMesh_1 Out TransformPoints_1 In)
  (connect Branch_2 Output%20B TransformPoints_1 In)
  (connect TransformPoints_1 Out TransformPoints_2 In)
  (connect ExtractAttribute_0 Out WorldRayHitQuery_1 RayOrigin))
