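; PCG graph for a Draw Spline editor-mode tool that spawns spline meshes along a drawn spline.
; It uses either control points or distance-resampled points, optionally projected onto world surfaces via ray hits cast from the top of their combined bounds.
; It rebuilds the spline, preserving its closed-loop state, gates it on the Mesh parameter, then applies global position offset and scale before spawning with the chosen forward axis.
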(graph SplineMeshTool
  (params (Mesh :type (SoftObject /Script/Engine.StaticMesh)) (MeshForwardAxis :type (Enum /Script/Engine.ESplineMeshAxis)) (ResampleSpline :type Bool) (ResampleDistanceIncrements :type Double :default 5e+02) (FitToCurve :type Bool :default true) (GlobalPositionOffset :type Vector) (GlobalScaleFactor :type Vector :default (1.0 1.0 1.0)) (SnapToSurface :type Bool) (AlignToSurface :type Bool) (SnapToSurfaceMaxDistance :type Double :default 1e+03))
  (bind GetSplineData_0 (Get%20Spline%20Data))
  (bind AttributeCompareOp_1 (Attribute%20Compare%20Op :InA Mesh))
  (bind Branch_1 (Branch :bOutputToB-param ResampleSpline))
  (bind GetSplineControlPoints_0 (Get%20Spline%20Control%20Points))
  (bind SplineSampler_0 (Spline%20Sampler :SamplerParams (struct :Mode Distance :DistanceIncrement 5e+02 :bUnbounded true :PointSteepness 1.0) :bFitToCurve FitToCurve :DistanceIncrement ResampleDistanceIncrements))
  (bind CombinePoints_1 (Combine%20Points :bUseFirstPointTransform false))
  (bind AttributeMathsOp_0 (Attribute%20Maths%20Op :InputSource1 $Position.z :InputSource2 $BoundsMax.Z))
  (bind ExtractAttribute_1 (Extract%20Attribute%20at%20Index :bForceOutputAttributeToBeInElementsDomain true :InputSource $Position))
  (bind WorldRayHitQuery_0 (World%20Ray%20Hit%20Query :QueryParams (struct :bOverrideDefaultParams true :RayLength 1e+03) :RayLength SnapToSurfaceMaxDistance))
  (bind Projection_0 (Projection :bKeepZeroDensityPoints true :ProjectionParams (struct :AttributeMode IncludeAttributes :AttributeMergeOperation SourceValue) :bProjectPositions SnapToSurface :bProjectRotations AlignToSurface))
  (bind CreateSpline_0 (Create%20Spline :InterpTypeAttribute None))
  (bind Branch_0 (Branch))
  (bind AddAttribute_0 (Add%20Attribute :InputSource Mesh :OutputTarget Mesh :Attributes Mesh))
  (bind AttributeMathsOp_2 (Attribute%20Maths%20Op :InputSource1 $Position :InB GlobalPositionOffset))
  (bind AttributeMathsOp_3 (Attribute%20Maths%20Op :InputSource1 $Scale :Operation Multiply :InB GlobalScaleFactor))
  (bind SpawnSplineMesh_0 (Spawn%20Spline%20Mesh :SplineMeshOverrideDescriptions [(struct :InputSource Mesh :PropertyTarget "StaticMesh")] :ForwardAxis MeshForwardAxis))
  (bind ExtractAttribute_0 (Extract%20Attribute%20at%20Index :bForceOutputAttributeToBeInElementsDomain true :InputSource @Data.$IsClosed :OutputAttributeName ClosedSpline))
  (reroute ClosedSpline ExtractAttribute_0)
  (connect Branch_0 Output%20A AddAttribute_0 In)
  (connect CombinePoints_1 Out AttributeMathsOp_0 InA)
  (connect CombinePoints_1 Out AttributeMathsOp_0 InB)
  (connect AddAttribute_0 Out AttributeMathsOp_2 InA)
  (connect AttributeMathsOp_2 Out AttributeMathsOp_3 InA)
  (connect AttributeCompareOp_1 Out Branch_0 bOutputToB)
  (connect CreateSpline_0 Out Branch_0 In)
  (connect GetSplineData_0 Out Branch_1 In)
  (connect GetSplineControlPoints_0 Out CombinePoints_1 In)
  (connect SplineSampler_0 Out CombinePoints_1 In)
  (connect ClosedSpline Out CreateSpline_0 bClosedLoop)
  (connect Projection_0 Out CreateSpline_0 In)
  (connect GetSplineData_0 Out ExtractAttribute_0 In)
  (connect AttributeMathsOp_0 Out ExtractAttribute_1 In)
  (connect Branch_1 Output%20A GetSplineControlPoints_0 In)
  (connect GetSplineControlPoints_0 Out Projection_0 In)
  (connect SplineSampler_0 Out Projection_0 In)
  (connect WorldRayHitQuery_0 Out Projection_0 Projection%20Target)
  (connect AttributeMathsOp_3 Out SpawnSplineMesh_0 In)
  (connect Branch_1 Output%20B SplineSampler_0 Spline)
  (connect ExtractAttribute_1 Out WorldRayHitQuery_0 RayOrigin))
